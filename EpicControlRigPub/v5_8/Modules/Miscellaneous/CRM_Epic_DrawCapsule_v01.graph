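# Control Rig module that draws a toggleable debug capsule guide on an animated rig.
# Construction resolves the Root connector, adds in-game and front-end bool channels when it is a control, and spawns a "Front End Capsule" shape control.
# Forward solve sets that control's visibility from the front-end channel.
# The in-game channel enables three orthogonal foreground debug circles, forming a wire sphere raised by the radius plus a location offset.
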ControlRigBlueprint /EpicControlRigPub/v5_8/Modules/Miscellaneous/CRM_Epic_DrawCapsule_v01
graphs [
 RigVMFunctionLibrary {links [], nodes []}
 RigVMFunction_Sequence_1_ContainedGraph {nodes [
   n1 = RigVMFunctionEntryNode()
   RigVMFunction_Sequence = &RigVMFunction_Sequence(ExecuteContext=@n1.ExecuteContext)
   n2 = &RigVMFunction_Sequence(ExecuteContext=@RigVMFunction_Sequence.B)
   RigVMFunctionReturnNode(A=@RigVMFunction_Sequence.A, B=@n2.A, C=@n2.B)
  ]}
 RigVMModel {nodes [
   n1 = &RigUnit_BeginExecution()
   RigVMFunction_Sequence = &RigVMFunction_Sequence(ExecuteContext=@n1.ExecutePin)
   SetControlVisibility = /Script/ControlRig.RigUnit_SetControlVisibility(
    bVisible=/Script/ControlRig.RigUnit_GetBoolAnimationChannel(
     bInitial=false
     Control=RigVMVariableNode() {variable=Root Connector}.Value.Name
     Channel=RigVMVariableNode() {variable=Front End Animation Channel}.Value.Name
    ) {is=struct RigVMUnitNode, template="GetAnimationChannel::Execute(out Value,in Control,in Channel,in bInitial)"}.Value
    Item=RigVMVariableNode() {variable=Front End Capsule}.Value
    ExecutePin=@RigVMFunction_Sequence.B
   ) {is=struct RigVMUnitNode, template="SetControlVisibility::Execute(in Item,in Pattern,in bVisible)"}
   n2 = RigVMVariableNode() {variable=Capsule Radius}
   RigVMFunction_MathTransformTransformVector = /Script/RigVM.RigVMFunction_MathTransformTransformVector(
    Transform=(Rotation=(X=0,Y=0,Z=0,W=1),Translation=(X=0,Y=0,Z=0),Scale3D=(X=1,Y=1,Z=1))
    Transform.Translation.Z=@n2.Value
    Location=RigVMVariableNode() {variable=Capsule Location Offset}.Value
   ) {is=struct RigVMUnitNode}
   n3 = RigVMVariableNode() {variable=Capsule Color}
   GetAnimationChannel = &RigUnit_GetBoolAnimationChannel(
    bInitial=false
    Control=RigVMVariableNode() {variable=Root Connector}.Value.Name
    Channel=RigVMVariableNode() {variable=In Game Animation Channel}.Value.Name
   ) {template="GetAnimationChannel::Execute(out Value,in Control,in Channel,in bInitial)"}
   n4 = RigVMVariableNode() {variable=Capsule Thickness}
   RigVMFunction_DebugArcNoSpace = /Script/RigVM.RigVMFunction_DebugArcNoSpace(
    DebugDrawSettings=(DepthPriority=SDPG_Foreground,Lifetime=-1)
    Detail=28
    MaximumDegrees=360
    MinimumDegrees=0
    Transform=(Rotation=(X=0,Y=0,Z=0,W=1),Translation=(X=0,Y=0,Z=0),Scale3D=(X=1,Y=1,Z=1))
    WorldOffset=(Rotation=(X=0,Y=0,Z=0,W=1),Translation=(X=0,Y=0,Z=0),Scale3D=(X=1,Y=1,Z=1))
    Transform.Translation=@RigVMFunction_MathTransformTransformVector.Result
    Color=@n3.Value
    bEnabled=@GetAnimationChannel.Value
    Radius=@n2.Value
    Thickness=@n4.Value
    ExecutePin=@RigVMFunction_Sequence.A
   ) {is=struct RigVMUnitNode}
   > &RigVMFunction_DebugArcNoSpace(
    DebugDrawSettings=(DepthPriority=SDPG_Foreground,Lifetime=-1)
    Detail=28
    MaximumDegrees=360
    MinimumDegrees=0
    Transform=(Rotation=(X=0,Y=-0.7071,Z=0,W=0.7071),Translation=(X=0,Y=0,Z=0),Scale3D=(X=1,Y=1,Z=1))
    WorldOffset=(Rotation=(X=0,Y=0,Z=0,W=1),Translation=(X=0,Y=0,Z=0),Scale3D=(X=1,Y=1,Z=1))
    Transform.Translation=@RigVMFunction_MathTransformTransformVector.Result
    Color=@n3.Value
    bEnabled=@GetAnimationChannel.Value
    Radius=@n2.Value
    Thickness=@n4.Value
   )
   > &RigVMFunction_DebugArcNoSpace(
    DebugDrawSettings=(DepthPriority=SDPG_Foreground,Lifetime=-1)
    Detail=28
    MaximumDegrees=360
    MinimumDegrees=0
    Transform=(Rotation=(X=-0.7071,Y=0,Z=0,W=0.7071),Translation=(X=0,Y=0,Z=0),Scale3D=(X=1,Y=1,Z=1))
    WorldOffset=(Rotation=(X=0,Y=0,Z=0,W=1),Translation=(X=0,Y=0,Z=0),Scale3D=(X=1,Y=1,Z=1))
    Transform.Translation=@RigVMFunction_MathTransformTransformVector.Result
    Color=@n3.Value
    bEnabled=@GetAnimationChannel.Value
    Radius=@n2.Value
    Thickness=@n4.Value
   )
  ]}
 RigVMModel Construction Event {nodes [
   n1 = &RigUnit_PrepareForExecution()
   RigVMFunction_Sequence = RigVMAggregateNode(ExecuteContext=@n1.ExecutePin)
   SpawnControl = /Script/ControlRig.RigUnit_HierarchyAddControlVector(
    InitialValue=(X=0,Y=0,Z=0)
    Name=Front End Capsule
    OffsetSpace=LocalSpace
    OffsetTransform=(Rotation=(X=0,Y=0,Z=0,W=1),Translation=(X=0,Y=0,Z=0),Scale3D=(X=1,Y=1,Z=1))
    Parent=(Type=None,Name="None")
    Settings="(InitialSpace=LocalSpace,bIsPosition=True,Limits=(LimitX=(bMinimum=False,bMaximum=False),LimitY=(bMinimum=False,bMaximum=False),LimitZ=(bMinimum=False,bMaximum=False),MinValue=(X=0,Y=0,Z=0),MaxValue=(X=1,Y=1,Z=1),bDrawLimits=True),Shape=(bVisible=True,Name ...(truncated)"
    Settings.Shape.bVisible=&RigUnit_GetBoolAnimationChannel(
     bInitial=false
     Control=RigVMVariableNode() {variable=Root Connector}.Value.Name
     Channel=RigVMVariableNode() {variable=Front End Animation Channel}.Value.Name
    ) {template="GetAnimationChannel::Execute(out Value,in Control,in Channel,in bInitial)"}.Value
    Settings.Shape.Transform=RigVMVariableNode() {variable=Capsule Transform}.Value
    ExecutePin=@RigVMFunction_Sequence.C
   ) {
    is=struct RigVMUnitNode
    template="SpawnControl::Execute(in InitialValue,in Settings,in OffsetTransform,in OffsetSpace,in Parent,in Name,out Item)"
   }
   > RigVMVariableNode(Value=@SpawnControl.Item) {variable=Front End Capsule}
   n2 = /Script/ControlRig.RigUnit_ResolveConnector(Connector=(Type=Connector,Name="Root"), SkipSocket=False) {is=struct RigVMUnitNode}
   n3 = RigVMVariableNode(Value=@n2.Result, ExecuteContext=@RigVMFunction_Sequence.A) {variable=Root Connector}
   RigVMFunction_ControlFlowBranch = &RigVMFunction_ControlFlowBranch(
    Condition=RigVMDispatchNode(B=Control, A=@n2.Result.Type) {template="DISPATCH_RigVMDispatch_CoreEquals(in A,in B,out Result)"}.Result
    ExecuteContext=@n3.ExecuteContext
   )
   SpawnAnimationChannel = /Script/ControlRig.RigUnit_HierarchyAddAnimationChannelBool(
    LimitsEnabled=()
    MaximumValue=True
    MinimumValue=False
    Name=Draw in Game Capsule
    ExecutePin=@RigVMFunction_ControlFlowBranch.True
    InitialValue=RigVMVariableNode() {variable=Inital State}.Value
    Parent=RigVMVariableNode() {variable=Root Connector}.Value
   ) {
    is=struct RigVMUnitNode
    template="SpawnAnimationChannel::Execute(in InitialValue,in MinimumValue,in MaximumValue,in LimitsEnabled,in Parent,in Name,out Item)"
   }
   > RigVMVariableNode(Value=@SpawnAnimationChannel.Item) {variable=In Game Animation Channel}
   > n4 = &RigUnit_HierarchyAddAnimationChannelBool(
    LimitsEnabled=()
    MaximumValue=True
    MinimumValue=False
    Name=Draw Front End Capsule
    Parent=RigVMVariableNode() {variable=Root Connector}.Value
    InitialValue=RigVMVariableNode() {variable=Inital State}.Value
   ) {template="SpawnAnimationChannel::Execute(in InitialValue,in MinimumValue,in MaximumValue,in LimitsEnabled,in Parent,in Name,out Item)"}
   > RigVMVariableNode(Value=@n4.Item) {variable=Front End Animation Channel}
  ]}
]
hierarchy [
 Root {type=Connector, loc=<0 0 0>, rot=<0 0 0 1>}
 Root {type=Socket, loc=<0 0 0>, rot=<0 0 0 1>}
]
variables [
 Capsule Color {type=FLinearColor, default=(R=0.1,G=0.4,B=0.7,A=1)}
 Capsule Location Offset {type=FVector, default=(X=0,Y=0,Z=0)}
 Capsule Radius {type=float, default=60}
 Capsule Thickness {type=float, default=1}
 Capsule Transform {type=FTransform, default=(Rotation=(X=0,Y=0,Z=0,W=1),Translation=(X=0,Y=0,Z=35),Scale3D=(X=1,Y=1,Z=1))}
 Front End Animation Channel {type=FRigElementKey, default=(Type=Bone,Name="")}
 Front End Capsule {type=FRigElementKey, default=(Type=Bone,Name="")}
 In Game Animation Channel {type=FRigElementKey, default=(Type=Bone,Name="")}
 Inital State {type=bool, default=False}
 Parent Control Connector {type=FRigElementKey, default=(Type=None,Name="")}
 Root Connector {type=FRigElementKey, default=(Type=Connector,Name="")}
]
