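# Control Rig module that pins Driven elements to Drivers, copying each driver's current transform onto its matching element.
# Construction resets both arrays and rebuilds them from the Drivers and Driven array connectors, keeping index-matched pairs only where the driver exists.
# Forward solve loops over the drivers and sets each paired driven element's transform at full weight, without propagating to children.
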ControlRigBlueprint /EpicControlRigPub/v5_8/Modules/Miscellaneous/CRM_Pin_v01
graphs [
 ConstructionGraph {nodes [
   RigUnit_PrepareForExecution = &RigUnit_PrepareForExecution()
   RigVMDispatchNode(ExecuteContext=@RigUnit_PrepareForExecution.ExecutePin, Array=RigVMVariableNode() {variable=Drivers}.Value) {template="DISPATCH_RigVMDispatch_ArrayReset(io Array)"}
   > RigVMDispatchNode(Array=RigVMVariableNode() {variable=Driven}.Value) {template="DISPATCH_RigVMDispatch_ArrayReset(io Array)"}
   > For_Each = RigVMDispatchNode(Array=/Script/ControlRig.RigUnit_ResolveArrayConnector(Connector=(Type=Connector,Name="Drivers"), SkipSocket=False) {is=struct RigVMUnitNode}.Result) {template="DISPATCH_RigVMDispatch_ArrayIterator(in Array,out Element,out Index,out Count,out Ratio)"}
   RigVMFunction_ControlFlowBranch = &RigVMFunction_ControlFlowBranch(
    ExecuteContext=@For_Each.ExecuteContext
    Condition=/Script/ControlRig.RigUnit_ItemExists(Item=@For_Each.Element) {is=struct RigVMUnitNode}.Exists
   )
   RigVMDispatchNode(ExecuteContext=@RigVMFunction_ControlFlowBranch.True, Array=RigVMVariableNode() {variable=Drivers}.Value, Element=@For_Each.Element) {template="DISPATCH_RigVMDispatch_ArrayAdd(io Array,in Element,out Index)"}
   > RigVMDispatchNode(
    Element=RigVMDispatchNode(
     Array=&RigUnit_ResolveArrayConnector(Connector=(Type=Connector,Name="Driven"), SkipSocket=False).Result
     Index=@For_Each.Index
    ) {template="DISPATCH_RigVMDispatch_ArrayGetAtIndex(in Array,in Index,out Element)"}.Element
    Array=RigVMVariableNode() {variable=Driven}.Value
   ) {template="DISPATCH_RigVMDispatch_ArrayAdd(io Array,in Element,out Index)"}
  ]}
 RigVMFunctionLibrary {links [], nodes []}
 RigVMModel {nodes [
   RigUnit_BeginExecution = &RigUnit_BeginExecution()
   For_Each = RigVMDispatchNode(ExecuteContext=@RigUnit_BeginExecution.ExecutePin, Array=RigVMVariableNode() {variable=Drivers}.Value) {template="DISPATCH_RigVMDispatch_ArrayIterator(in Array,out Element,out Index,out Count,out Ratio)"}
   > &RigUnit_SetTransform(
    Weight=1
    bInitial=False
    bPropagateToChildren=false
    Item=RigVMDispatchNode(Index=@For_Each.Index, Array=RigVMVariableNode() {variable=Driven}.Value) {template="DISPATCH_RigVMDispatch_ArrayGetAtIndex(in Array,in Index,out Element)"}.Element
    Value=&RigUnit_GetTransform(bInitial=False, Item=@For_Each.Element).Transform
   )
  ]}
]
hierarchy [
 Root {type=Connector, loc=<0 0 0>, rot=<0 0 0 1>}
 Root {type=Socket, loc=<0 0 0>, rot=<0 0 0 1>}
 Drivers {type=Connector, loc=<0 0 0>, rot=<0 0 0 1>}
 Driven {type=Connector, loc=<0 0 0>, rot=<0 0 0 1>}
]
variables [
 Driven {type="TArray<FRigElementKey>", default=((Type=Bone,Name="ik_hand_l"))}
 Drivers {type="TArray<FRigElementKey>", default=((Type=Bone,Name="hand_l"))}
]
